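; Burst gameplay cue for the bush grenade impact that customizes the leaf effect it spawns.
; OnBurst calls the parent implementation, then takes the first spawned particle component and checks that it is valid.
; It copies the cue's default variables into that Niagara system's user parameters, setting leaf color and amount plus the offset, height and radius of a tree cylinder.
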(fn OnBurst(EitherVFX-preferthisversion)
  (params (MyTarget :type Actor%20Object%20Reference) (Parameters :type Gameplay%20Cue%20Parameters%20Structure%20%28by%20ref%29) (ParticleComponents :type Array%20of%20FXSystem%20Component%20Object%20References) (AudioComponents :type Array%20of%20Audio%20Component%20Object%20References) (BurstCameraShakeInstance :type Camera%20Shake%20Base%20Object%20Reference) (BurstDecalInstance :type Decal%20Actor%20Object%20Reference))
  (bind _output (Utilities|Array|Get(acopy) ParticleComponents))
  ;[1 1 1] OnBurst-
  ; Adjusting Niagara System's parameters
  (|CallParentFunction MyTarget Parameters ParticleComponents AudioComponents BurstCameraShakeInstance BurstDecalInstance)
  (Utilities|IsValid _output
    (:Is%20Valid
      (Effects|Components|ParticleSystem|SetColorParameter _output "LeafBaseColor" (Variables|Default|GetLeafBaseColorParam))
      (Effects|Components|ParticleSystem|SetVectorParameter _output "TreeCylinderOffset" (Variables|Default|GetTreeCylinderOffsetParam))
      (Effects|Components|ParticleSystem|SetFloatParameter _output "LeafAmount" (Variables|Default|GetLeafAmountParam))
      (Effects|Components|ParticleSystem|SetFloatParameter _output "TreeCylinderHeight" (Variables|Default|GetTreeCylinderHeightParam))
      (Effects|Components|ParticleSystem|SetFloatParameter _output "TreeCylinderRadius" (Variables|Default|GetTreeCylinderRadiusParam))
      (return))
    (:Is%20Not%20Valid
      (return))))
